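# Niagara effect for the Lifting Aura emote, apparently on its ropes, that emits sparkles and small glowing flecks along a skeletal mesh.
# Both active emitters split particles into ten partitions, using the index to pick a bone or socket, then scatter them in a stretched sphere.
# GPU RopeSparkles are warm and fall under ramping gravity, while Snowflakes005 drifts upward with curl noise and a pink color-over-life tint.
# The larger, orange-tinted Snowflakes emitter is disabled.
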NiagaraSystem /BRCosmetics/Effects/Fort_Effects/Emotes/LiftingAura/NS_Emote_LiftingAura_Ropes
dependencies {
 usedDataInterfaces [
  &NiagaraDataInterfaceCurve
  &NiagaraDataInterfaceMeshRendererInfo
  &NiagaraDataInterfaceColorCurve
  &NiagaraDataInterfaceSkeletalMesh
 ]
 usedDynamicInputs []
 usedModules [
  &self:SystemSpawnScript.SystemScriptSource.SystemScriptGraph.NiagaraNodeAssignment_0.SetVariables_#1
  &SystemState
 ]
}
emitters {
 RopeSparkles {
  simTarget=GPUComputeSim
  emitterUpdateScript [
   &EmitterState(
    Inactive Response=::NewEnumerator1
    Life Cycle Mode=::NewEnumerator1
    Loop Behavior=::NewEnumerator1
    Loop Duration Mode=::NewEnumerator1
    UseLoopDelay=false
   )
   &SpawnRate(SpawnRate=20, Use Spawn Probability=false)
  ]
  particleSpawnScript [
   &InitializeParticle(
    Color=<rgba 5 3.7043 1.51 1>
    Lifetime=&Multiply_Float(A=&RandomRangeFloat(Minimum=0.05, Maximum=1), B=1)
    Sprite Size Mode=::NewEnumerator1
    UsePositionOffset=false
    Write Material Random=true
   )
   &PartitionParticles(Number Of Partitions=10, Partition Mode:ENiagara_PartitionMode=::NewEnumerator5)
   &SkeletalMeshLocation(
    Bone / Socket Index=$Particles.PartitionParticles.Partition
    Bone Sampling Mode=::NewEnumerator4
    MaskSpawnGroup=false
    Skeletal Mesh=$System.Skeletal Mesh
   )
   &ShapeLocation(MaskSpawnGroup=false, Non Uniform Scale=<1 1 1.5>, Offset Mode=::NewEnumerator0, Override Seed=false, Sphere Radius=10, UseOwnerScale=true)
   &CameraOffset(Camera Offset Amount=50) {enabled=false}
  ]
  particleUpdateScript [
   &ParticleState()
   &GravityForce(Gravity=&Multiply_VectorByFloat(
     Vector=<1 1 -98>
     Float=&FloatFromCurve(FloatCurve=DataInterface() {propertyValues {Curve {keys [
          {arriveTangent=2.2957, interpMode=RCIM_Cubic, leaveTangent=2.2957, tangentMode=RCTM_User, time=0, value=0}
          {arriveTangent=1.1292, interpMode=RCIM_Cubic, leaveTangent=1.1292, tangentMode=RCTM_User, time=1, value=1}
         ]}}})
    ))
   &CurlNoiseForce(Noise Strength=25, Pan Noise=false)
   &Drag(Use Linear Drag=true, Use Rotational Drag=true)
   &SolveForcesAndVelocity(Clamp Velocity=false, Limit Acceleration=false)
   &ScaleColor(
    Scale Alpha=&FloatFromCurve(FloatCurve=DataInterface() {propertyValues {Curve {keys [<at -0.0092 0.0144>,<at 0.0967 0.9767>,<at 1.0011 0.002>]}}})
    ScaleA=true
    ScaleRGB=true
   )
   &ScaleSpriteSize(Scale Factor=&Vector2DFromFloat(Value=&FloatFromCurve(FloatCurve=DataInterface() {propertyValues {Curve {keys [<at -0.0046 0>,<at 0.1746 0.9952>,<at 1.0011 0.002>]}}}, Scale Curve=1.5)))
  ]
  rendererClasses [&NiagaraSpriteRendererProperties]
  renderers [{rendererClass=&NiagaraSpriteRendererProperties}]
 }
 Snowflakes {
  bEnabled=false
  emitterUpdateScript [
   &EmitterState(
    Inactive Response=::NewEnumerator1
    Life Cycle Mode=::NewEnumerator1
    Loop Behavior=::NewEnumerator1
    Loop Duration Mode=::NewEnumerator1
    UseLoopDelay=false
   )
   &SpawnRate(SpawnRate=5, Use Spawn Probability=false)
   &SpawnPerUnit(Module.Use Max Movement Threshold=true, Module.Use Movement Tolerance=false, Spawn Spacing=10, Use Spawn Probability=false) {enabled=false}
  ]
  particleSpawnScript [
   &InitializeParticle(
    Color=&FNTimeOfDayColor(Night Color=<rgba 0.75 0.75 0.75 1>, Morning Color=<rgba 0.75 0.75 0.75 1>, Day Color=<rgba 1 1 1 1>, Evening Color=<rgba 0.75 0.75 0.75 1>)
    Lifetime Max=0.9
    Lifetime Min=0.7
    Lifetime Mode=::NewEnumerator1
    Mass Mode=::NewEnumerator1
    Sprite Size Mode=::NewEnumerator1
    Uniform Sprite Size Max=75
    Uniform Sprite Size Min=50
    UsePositionOffset=false
    Write Material Random=true
   ) {module=InitializeParticle001}
   &PartitionParticles(Number Of Partitions=10, Partition Mode:ENiagara_PartitionMode=::NewEnumerator5)
   &SkeletalMeshLocation(
    Bone / Socket Index=$Particles.PartitionParticles.Partition
    Bone Sampling Mode=::NewEnumerator4
    MaskSpawnGroup=false
    Skeletal Mesh=$System.Skeletal Mesh
   )
   &ShapeLocation(MaskSpawnGroup=false, Non Uniform Scale=<1 1 1.5>, Offset Mode=::NewEnumerator0, Override Seed=false, Sphere Radius=10, UseOwnerScale=true) {module=ShapeLocation001}
   /SetVariables(Particles.UVRandom:NiagaraFloat=&RandomRangeFloat(Minimum=0, Maximum=3)) {isSetParameters=true, module=SetVariables_#2}
   /SetVariables(Particles.SubImageIndex=&RandomRangeFloat(Minimum=0, Maximum=4)) {enabled=false, isSetParameters=true, module=SetVariables_#3}
   &ApplyInitialForces(Apply Force to Position=true)
   /SetVariables(Particles.SpriteScale:Vector2f=&Multiply_Vector2D(A=$Particles.SpriteSize, B=<0.95 0.95>)) {isSetParameters=true, module=SetVariables_#4}
  ]
  particleUpdateScript [
   &ParticleState()
   /SetVariables(Particles.ScaleSpriteSize_02:Vector2f=&Multiply_Vector2D_ByFloat(
     Float=&FloatFromCurve(
      FloatCurve=DataInterface() {propertyValues {Curve {keys [
          {interpMode=RCIM_Cubic, tangentMode=RCTM_User, time=0, value=0}
          {interpMode=RCIM_Cubic, tangentMode=RCTM_User, time=0.4, value=1}
          {tangentMode=RCTM_User, time=1, value=0}
         ]}}}
      Scale Curve=2.25
     )
     Vector2D=$Particles.SpriteScale
    )) {isSetParameters=true, module=SetVariables_#5}
   &ScaleSpriteSize(Uniform Curve Sprite Scale=DataInterface() {propertyValues {Curve {keys [
        {interpMode=RCIM_Cubic, tangentMode=RCTM_User, time=0, value=0}
        {interpMode=RCIM_Cubic, tangentMode=RCTM_User, time=0.4, value=1}
        {tangentMode=RCTM_User, time=1, value=0}
       ]}}}) {module=ScaleSpriteSize004}
   /Niagara/Modules/Update/Forces/CurlNoiseForce(Noise Strength=&RandomRangeFloat(Minimum=120, Maximum=160), Pan Noise=false)
   &GravityForce(Gravity=&RandomRangeVector(Minimum=<0 0 25>, Maximum=<0 0 50>))
   &Drag(Drag=&UniformRangedFloat(Minimum=6, Maximum=8), Use Linear Drag=true, Use Rotational Drag=false)
   &SolveForcesAndVelocity(Clamp Velocity=false, Limit Acceleration=false)
   &DynamicMaterialParameters(
    Index 0 Param 1=&RandomRangeFloat(Minimum=4, Maximum=8)
    Module.Index0Param0Write=true
    Module.Index0Param1Write=true
    Module.Index0Param2Write=true
    Module.Index0Param3Write=true
   ) {module=DynamicMaterialParameters001}
   &ScaleColor(
    Scale Alpha=&FloatFromCurve(FloatCurve=DataInterface() {propertyValues {Curve {keys [
         {arriveTangent=9.8203, interpMode=RCIM_Cubic, leaveTangent=9.8203, tangentMode=RCTM_User, time=0, value=1}
         {arriveTangent=-2.5249, interpMode=RCIM_Cubic, leaveTangent=-2.5249, tangentMode=RCTM_User, time=0.5, value=0.9847}
         {arriveTangent=-0.0405, interpMode=RCIM_Cubic, leaveTangent=-0.0405, tangentMode=RCTM_User, time=1, value=0}
        ]}}})
    ScaleA=true
    ScaleRGB=true
   ) {module=ScaleColor001}
   /SetVariables(Particles.SpriteColor=&MultiplyLinearColor(
     Linear Color=&FNTimeOfDayColor(Night Color=<rgba 1 1 1 1>, Morning Color=<rgba 1 1 1 1>, Day Color=<rgba 1 1 1 1>, Evening Color=<rgba 1 1 1 1>)
     Scale Factor=&MakeVector4_FromLinearColor(LinearColor=&ColorFromCurve(ColorCurve=DataInterface() {propertyValues {
         AlphaCurve {keys [
           {arriveTangent=2.6759, interpMode=RCIM_Cubic, leaveTangent=2.6759, tangentMode=RCTM_User, time=0, value=0.6}
           {arriveTangent=-1.7996, interpMode=RCIM_Cubic, leaveTangent=-2.7184, tangentMode=RCTM_Break, time=0.5, value=0.6}
           <at 1 0 RCIM_Cubic>
          ]}
         BlueCurve {keys [<at 0 0.5105 RCIM_Cubic>,<at 0.4702 0.2864>,<at 1 0.0604 RCIM_Cubic>]}
         GreenCurve {keys [<at 0 0.391 RCIM_Cubic>,<at 0.4702 0.2266>,<at 1 0.0396 RCIM_Cubic>]}
         RedCurve {keys [<at 0 2 RCIM_Cubic>,<at 0.4702 1>,<at 1 0.2604 RCIM_Cubic>]}
        }}))
    )) {isSetParameters=true, module=SetVariables_#6}
   /SetVariables(Particles.DynamicPar:Vector4f=&MakeVector4(X=0, Y=2, Z=0, W=0)) {isSetParameters=true, module=SetVariables_#7}
  ]
  rendererClasses [&NiagaraSpriteRendererProperties]
  renderers [
   {rendererClass=&NiagaraSpriteRendererProperties}
   {rendererClass=&NiagaraSpriteRendererProperties, rendererIndex=1}
  ]
 }
 Snowflakes005 {
  emitterUpdateScript [
   &EmitterState(
    Inactive Response=::NewEnumerator1
    Life Cycle Mode=::NewEnumerator1
    Loop Behavior=::NewEnumerator1
    Loop Duration Mode=::NewEnumerator1
    UseLoopDelay=false
   )
   &SpawnRate(SpawnRate=5, Use Spawn Probability=false)
   &SpawnPerUnit(Module.Use Max Movement Threshold=true, Module.Use Movement Tolerance=false, Spawn Spacing=10, Use Spawn Probability=false) {enabled=false}
  ]
  particleSpawnScript [
   &InitializeParticle(
    Color=&FNTimeOfDayColor(Night Color=<rgba 0.75 0.75 0.75 1>, Morning Color=<rgba 0.75 0.75 0.75 1>, Day Color=<rgba 1 1 1 1>, Evening Color=<rgba 0.75 0.75 0.75 1>)
    Lifetime Max=0.9
    Lifetime Min=0.7
    Lifetime Mode=::NewEnumerator1
    Mass Mode=::NewEnumerator1
    Sprite Size Mode=::NewEnumerator1
    Uniform Sprite Size Max=7.5
    UsePositionOffset=false
    Write Material Random=true
   ) {module=InitializeParticle001}
   &PartitionParticles(Number Of Partitions=10, Partition Mode:ENiagara_PartitionMode=::NewEnumerator5)
   &SkeletalMeshLocation(
    Bone / Socket Index=$Particles.PartitionParticles.Partition
    Bone Sampling Mode=::NewEnumerator4
    MaskSpawnGroup=false
    Skeletal Mesh=$System.Skeletal Mesh
   )
   &ShapeLocation(MaskSpawnGroup=false, Non Uniform Scale=<1 1 1.5>, Offset Mode=::NewEnumerator0, Override Seed=false, Sphere Radius=10, UseOwnerScale=true) {module=ShapeLocation001}
   /SetVariables(Particles.UVRandom:NiagaraFloat=&RandomRangeFloat(Minimum=0, Maximum=3)) {isSetParameters=true, module=SetVariables_#2}
   /SetVariables(Particles.SubImageIndex=&RandomRangeFloat(Minimum=0, Maximum=4)) {enabled=false, isSetParameters=true, module=SetVariables_#3}
   &ApplyInitialForces(Apply Force to Position=true)
   /SetVariables(Particles.SpriteScale:Vector2f=&Multiply_Vector2D(A=$Particles.SpriteSize, B=<0.95 0.95>)) {isSetParameters=true, module=SetVariables_#4}
  ]
  particleUpdateScript [
   &ParticleState()
   /SetVariables(Particles.ScaleSpriteSize_02:Vector2f=&Multiply_Vector2D_ByFloat(
     Float=&FloatFromCurve(
      FloatCurve=DataInterface() {propertyValues {Curve {keys [
          {interpMode=RCIM_Cubic, tangentMode=RCTM_User, time=0, value=0}
          {interpMode=RCIM_Cubic, tangentMode=RCTM_User, time=0.4, value=1}
          {tangentMode=RCTM_User, time=1, value=0}
         ]}}}
      Scale Curve=2.25
     )
     Vector2D=$Particles.SpriteScale
    )) {isSetParameters=true, module=SetVariables_#5}
   &ScaleSpriteSize(Uniform Curve Sprite Scale=DataInterface() {propertyValues {Curve {keys [
        {interpMode=RCIM_Cubic, tangentMode=RCTM_User, time=0, value=0}
        {interpMode=RCIM_Cubic, tangentMode=RCTM_User, time=0.4, value=1}
        {tangentMode=RCTM_User, time=1, value=0}
       ]}}}) {module=ScaleSpriteSize004}
   &CurlNoiseForce(Noise Strength=&RandomRangeFloat(Minimum=120, Maximum=160), Pan Noise=false)
   &GravityForce(Gravity=&RandomRangeVector(Minimum=<0 0 25>, Maximum=<0 0 50>))
   &Drag(Drag=&UniformRangedFloat(Minimum=6, Maximum=8), Use Linear Drag=true, Use Rotational Drag=false)
   &SolveForcesAndVelocity(Clamp Velocity=false, Limit Acceleration=false)
   &DynamicMaterialParameters(Module.Index0Param0Write=true, Module.Index0Param1Write=true, Module.Index0Param2Write=true, Module.Index0Param3Write=true) {module=DynamicMaterialParameters001}
   &ScaleColor(
    Scale Alpha=&FloatFromCurve(FloatCurve=DataInterface() {propertyValues {Curve {keys [
         {arriveTangent=9.8203, interpMode=RCIM_Cubic, leaveTangent=9.8203, tangentMode=RCTM_User, time=0, value=1}
         {arriveTangent=-2.5249, interpMode=RCIM_Cubic, leaveTangent=-2.5249, tangentMode=RCTM_User, time=0.5, value=0.9847}
         {arriveTangent=-0.0405, interpMode=RCIM_Cubic, leaveTangent=-0.0405, tangentMode=RCTM_User, time=1, value=0}
        ]}}})
    ScaleA=true
    ScaleRGB=true
   ) {module=ScaleColor001}
   /SetVariables(Particles.SpriteColor=&MultiplyLinearColor(
     Linear Color=&FNTimeOfDayColor(Night Color=<rgba 1 1 1 1>, Morning Color=<rgba 1 1 1 1>, Day Color=<rgba 1 1 1 1>, Evening Color=<rgba 1 1 1 1>)
     Scale Factor=&MakeVector4_FromLinearColor(LinearColor=&ColorFromCurve(ColorCurve=DataInterface() {propertyValues {
         AlphaCurve {keys [
           {arriveTangent=2.6759, interpMode=RCIM_Cubic, leaveTangent=2.6759, tangentMode=RCTM_User, time=0, value=0.6}
           {arriveTangent=-1.7996, interpMode=RCIM_Cubic, leaveTangent=-2.7184, tangentMode=RCTM_Break, time=0.5, value=0.6}
           <at 1 0 RCIM_Cubic>
          ]}
         BlueCurve {keys [<at 0 1.6627 RCIM_Cubic>,<at 0.4702 0.4582>,<at 1 0.0604 RCIM_Cubic>]}
         GreenCurve {keys [<at 0 0.482 RCIM_Cubic>,<at 0.4702 0>,<at 1 0.0396 RCIM_Cubic>]}
         RedCurve {keys [<at 0 2 RCIM_Cubic>,<at 0.4702 1>,<at 1 0.2604 RCIM_Cubic>]}
        }}))
    )) {isSetParameters=true, module=SetVariables_#6}
   /SetVariables(Particles.DynamicPar:Vector4f=&MakeVector4(X=0, Y=2, Z=0, W=0)) {isSetParameters=true, module=SetVariables_#7}
  ]
  rendererClasses [&NiagaraSpriteRendererProperties]
  renderers [
   {rendererClass=&NiagaraSpriteRendererProperties}
   {rendererClass=&NiagaraSpriteRendererProperties, rendererIndex=1}
  ]
 }
}
summary {
 systemName=NS_Emote_LiftingAura_Ropes
 emitters [
  {emitterName=RopeSparkles, simTarget=GPUComputeSim, rendererClasses [&NiagaraSpriteRendererProperties]}
  {emitterName=Snowflakes005, rendererClasses [&NiagaraSpriteRendererProperties]}
  {bEnabled=false, emitterName=Snowflakes, rendererClasses [&NiagaraSpriteRendererProperties]}
 ]
 userVariables []
}
